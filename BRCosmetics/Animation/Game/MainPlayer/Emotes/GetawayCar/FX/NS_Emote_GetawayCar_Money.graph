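# Niagara effect for the GetawayCar emote that spills money and dresses the car with smoke, glow and lights.
# Coins, as mesh or sprite, spawn by rate and distance moved at the vehicle, then tumble and collide, alongside bling sparkles.
# Wheel smoke, golden glow, light rays, and socket-attached head and tail lights complete the look.
# Front-end switches, time-of-day scalars and platform quality adjust spawning, color, collision and light count.
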NiagaraSystem /BRCosmetics/Animation/Game/MainPlayer/Emotes/GetawayCar/FX/NS_Emote_GetawayCar_Money
dependencies {
 usedDataInterfaces [
  &NiagaraDataInterfacePlatformSet
  &NiagaraDataInterfaceArrayInt32
  &NiagaraDataInterfaceCurve
  &NiagaraDataInterfaceColorCurve
  &NiagaraDataInterfaceSpriteRendererInfo
  &NiagaraDataInterfaceSkeletalMesh
 ]
 usedDynamicInputs []
 usedModules [
  &self:SystemSpawnScript.SystemScriptSource.SystemScriptGraph.NiagaraNodeAssignment_1.SetVariables_#1
  &self:SystemSpawnScript.SystemScriptSource.SystemScriptGraph.NiagaraNodeAssignment_2.SetVariables_#2
  &self:SystemSpawnScript.SystemScriptSource.SystemScriptGraph.NiagaraNodeAssignment_4.SetVariables_#3
  &SystemState
 ]
}
emitters {
 Coin {
  emitterUpdateScript [
   &EmitterState(
    Life Cycle Mode=::NewEnumerator1
    Loop Behavior=::NewEnumerator1
    Loop Delay=&RandomRangeFloat(Minimum=2, Maximum=4)
    Loop Duration Mode=::NewEnumerator1
    UseLoopDelay=false
   )
   &SpawnRate(SpawnRate=&FN_FrontEndFloatModifier(Gameplay=0, Front End=20), Use Spawn Probability=false)
   &SpawnRate(Spawn Probability=0.5, SpawnRate=3, Use Spawn Probability=true) {module=SpawnRate001}
   &SpawnPerUnit(
    Module.Use Max Movement Threshold=true
    Module.Use Movement Tolerance=false
    Spawn Probability=0.8
    Spawn Spacing=7
    Use Spawn Probability=true
   )
  ]
  particleSpawnScript [
   /SetVariables(Particles.VisibilityTag=&SelectIntFromArray(
     Direct Array Index=&RandomRangeInt(Minimum=0, Maximum=2)
     Int Selection Array=DataInterface() {propertyValues {IntData [0,1,1]}}
    )) {isSetParameters=true}
   &InitializeParticle(
    Lifetime Max=0.6
    Lifetime Mode=::NewEnumerator1
    Mesh Scale Mode=::NewEnumerator3
    Mesh Uniform Scale=&RandomRangeFloat(Minimum=0.05, Maximum=0.03)
    Sprite Rotation Mode=::NewEnumerator2
    Sprite Size Mode=::NewEnumerator3
    Uniform Sprite Size=5
    UsePositionOffset=false
    Write Material Random=true
   )
   &SkeletalMeshLocation(Bone Sampling Mode=::NewEnumerator4, MaskSpawnGroup=false, Sampled Position Offset=<0 -25 0>, Skeletal Mesh=$System.VehicleLocation)
   &ShapeLocation(
    MaskSpawnGroup=false
    Non Uniform Scale=<1 1 0.5>
    Offset=<0 0 -5>
    Offset Mode=::NewEnumerator0
    Override Seed=false
    Sphere Radius=30
    UseOwnerScale=true
   )
   &InitialMeshOrientation(Use Rotation Vector=false)
   /SetVariables(Particles.MeshColor:LinearColor=<rgba 1 0.5686 0.2157 0>) {isSetParameters=true}
   &AddVelocity(
    Velocity=&RandomRangeVector(Minimum=<-70 -200 50>, Maximum=<70 -100 80>)
    Velocity Speed Scale=&FN_FrontEndFloatModifier(Gameplay=0.2, Front End=1)
   )
  ]
  particleUpdateScript [
   /SetVariables(Particles.SineWave:NiagaraFloat=&Sine(Normalized Angle=$Emitter.Age, Period=&RandomRangeFloat(Minimum=0.3, Maximum=0.7), Scale=1, Bias=0)) {isSetParameters=true}
   &ParticleState()
   &ScaleSpriteSize(Scale Factor=&MakeVector2D(X=$Particles.SineWave, Y=1), Scale Sprite Size Mode=::NewEnumerator2) {module=ScaleSpriteSize001}
   &SpriteRotationRate(Rotation Rate=&RandomRangeFloat(Minimum=-1000, Maximum=-500))
   &ScaleMeshSize(Scale Factor=&VectorFromFloat(Value=&FloatFromCurve(FloatCurve=DataInterface() {propertyValues {Curve {keys [<at 0.6 1>,<at 1 0>]}}})))
   &ScaleSpriteSize(Uniform Curve Sprite Scale=DataInterface() {propertyValues {Curve {keys [<at 0.001 0.5>,<at 0.2948 1.1652>,<at 1.0009 0.8072>]}}})
   &ScaleColor(
    Scale Alpha=&FloatFromCurve(
     FloatCurve=DataInterface() {propertyValues {Curve {keys [
         {arriveTangent=-0.0888, interpMode=RCIM_Cubic, leaveTangent=-0.0888, tangentMode=RCTM_User, time=0, value=0}
         {arriveTangent=0.12, interpMode=RCIM_Cubic, leaveTangent=0.12, tangentMode=RCTM_User, time=0.15, value=0.9847}
         {arriveTangent=-0.0405, interpMode=RCIM_Cubic, leaveTangent=-0.0405, tangentMode=RCTM_User, time=1, value=0}
        ]}}}
     Scale Curve=&FNTimeOfDayScalar(
      Night Scale=1
      Morning Scale=1
      Day Scale=&FN_FrontEndFloatModifier(Gameplay=1, Front End=1)
      Evening Scale=1
     )
    )
    Scale RGB=&VectorFromFloat(Value=&FNTimeOfDayScalar(
      Night Scale=0.8
      Morning Scale=0.8
      Day Scale=&FN_FrontEndFloatModifier(Gameplay=1, Front End=1)
      Evening Scale=0.5
     ))
    ScaleA=true
    ScaleRGB=true
   ) {module=ScaleColor003}
   &Drag(
    Drag=&FloatFromCurve(
     FloatCurve=DataInterface() {propertyValues {Curve {keys [
         {arriveTangent=0.7001, interpMode=RCIM_Cubic, leaveTangent=0.7001, tangentMode=RCTM_User, time=0, value=0}
         {interpMode=RCIM_Cubic, tangentMode=RCTM_User, time=1, value=1}
        ]}}}
     Scale Curve=4
    )
    Use Linear Drag=true
    Use Rotational Drag=true
   ) {module=Drag002}
   &GravityForce()
   &UpdateMeshOrientation(Rotation Vector=&Multiply_Vector(
     A=<1 1 1>
     B=&MakeVector(
      X=&MakeCustomFloatFromBool(Bool=&RandomBool(), True Float=2, False Float=-2)
      Y=&RandomRangeFloat(Minimum=0.1, Maximum=-0.1)
      Z=&RandomRangeFloat(Minimum=0.1, Maximum=-0.1)
     )
    ))
   &Collision(
    Age Colliding Particles=false
    Collision Enabled=&Boolean_OrOperation(
     Bool1=&IsPlatformSetActive(New Platform Set=DataInterface() {propertyValues {Platforms {qualityLevelMask=-8}}})
     Bool2=&Boolean_NotOperation(Bool=$User.bIsFrontEndPreview)
    )
    KillOnCollision=false
    Module.Control Rest State Via Surface Normal=false
    Module.EnableMaxCollisionCount=false
    Module.Randomize Collision Normal Vector=false
    Radius Calculation Type=::NewEnumerator1
   )
   &Collision(
    Age Colliding Particles=false
    Analytical Collision Plane Position 1:Vector3f=<0 0 -50>
    Analytical Collision Plane Space:ENiagaraCoordinateSpace=::Local
    CPU Collision Type=::NewEnumerator3
    Collision Enabled=&Boolean_OrOperation(
     Bool1=&IsPlatformSetActive(New Platform Set=DataInterface() {propertyValues {Platforms {qualityLevelMask=-28}}})
     Bool2=$User.bIsFrontEndPreview
    )
    KillOnCollision=false
    Module.Control Rest State Via Surface Normal=false
    Module.EnableMaxCollisionCount=false
    Module.Randomize Collision Normal Vector=false
    Perform 2 Analytical Collision Queries=false
   ) {module=Collision001}
   &SolveForcesAndVelocity(Clamp Velocity=false, Limit Acceleration=false)
   &DynamicMaterialParameters(
    Index 0 Param 1=&Add_Float(
     A=&RandomRangeFloat(Minimum=0, Maximum=1)
     B=&FloatFromCurve(
      FloatCurve=DataInterface() {propertyValues {Curve {keys [
          {arriveTangent=2.7533, interpMode=RCIM_Cubic, leaveTangent=1.8958, tangentMode=RCTM_User, time=0, value=1}
          {arriveTangent=0.8292, interpMode=RCIM_Cubic, leaveTangent=0.8292, tangentMode=RCTM_User, time=0.321, value=1.3453}
          {arriveTangent=0.0529, interpMode=RCIM_Cubic, leaveTangent=0.2021, tangentMode=RCTM_User, time=1, value=1.5}
         ]}}}
      CurveIndex=&Add_Float(A=&Multiply_Float(A=1, B=0.5), B=$Particles.NormalizedAge)
     )
    )
    Index 0 Param 2=&FloatFromCurve(FloatCurve=DataInterface() {propertyValues {Curve {keys [
         {arriveTangent=0.0453, interpMode=RCIM_Cubic, tangentMode=RCTM_Break, time=0, value=-1}
         {arriveTangent=0.0604, interpMode=RCIM_Cubic, leaveTangent=0.0604, tangentMode=RCTM_User, time=1, value=-1}
        ]}}})
    Index 0 Param 3=&FloatFromCurve(FloatCurve=DataInterface() {propertyValues {Curve {keys [
         {arriveTangent=-53.9199, interpMode=RCIM_Cubic, leaveTangent=-8.9727, tangentMode=RCTM_User, time=0, value=1}
         {arriveTangent=0.2347, interpMode=RCIM_Cubic, leaveTangent=0.2723, tangentMode=RCTM_Break, time=0.25, value=0}
        ]}}})
    Index 0 Param 4=&FloatFromCurve(
     FloatCurve=DataInterface() {propertyValues {Curve {keys [<at 0.5 1>,<at 1 0>]}}}
     Scale Curve=&FNTimeOfDayScalar(
      Night Scale=2
      Morning Scale=2
      Day Scale=&FN_FrontEndFloatModifier(Gameplay=1, Front End=0.5)
      Evening Scale=3
     )
    )
    Index 2 Param 1=&FloatFromCurve(FloatCurve=DataInterface() {propertyValues {Curve {keys [
         {interpMode=RCIM_Cubic, leaveTangent=4.2485, tangentMode=RCTM_Break, time=0, value=0}
         {arriveTangent=1.0548, interpMode=RCIM_Cubic, leaveTangent=1.0548, tangentMode=RCTM_User, time=0.17, value=0.5715}
         <at 1 1>
        ]}}})
    Index 2 Param 2=&FloatFromCurve(
     FloatCurve=DataInterface() {propertyValues {Curve {keys [
         {arriveTangent=1.9056, interpMode=RCIM_Cubic, leaveTangent=1.9056, tangentMode=RCTM_User, time=0, value=0.2}
         {arriveTangent=0.0259, interpMode=RCIM_Cubic, leaveTangent=0.0259, tangentMode=RCTM_User, time=1, value=1.4}
        ]}}}
     Scale Curve=&RandomRangeFloat(Minimum=0.2, Maximum=0.5)
    )
    Index 2 Param 3=&FloatFromCurve(FloatCurve=DataInterface() {propertyValues {Curve {keys [
         {interpMode=RCIM_Cubic, leaveTangent=2.3064, tangentMode=RCTM_Break, time=0, value=1}
         {interpMode=RCIM_Cubic, leaveTangent=11.1233, tangentMode=RCTM_Break, time=1, value=2}
        ]}}})
    Index 2 Param 4=&FloatFromCurve(
     FloatCurve=DataInterface() {propertyValues {Curve {keys [
         {arriveTangent=-0.6867, interpMode=RCIM_Cubic, leaveTangent=-0.6867, tangentMode=RCTM_User, time=0, value=1}
         {interpMode=RCIM_Cubic, leaveTangent=-1.2956, tangentMode=RCTM_Break, time=1, value=0.75}
        ]}}}
     Scale Curve=1.6
    )
    Module.Index0Float0Write=true
    Module.Index0Float1Write=true
    Module.Index0Float2Write=true
    Module.Index0Float3Write=true
    Module.Index2Float0Write=true
    Module.Index2Float1Write=true
    Module.Index2Float2Write=true
    Module.Index2Float3Write=true
    Write Parameter Index 2=::NewEnumerator1
   )
   &SubUVAnimation(
    Evaluation Type=::NewEnumerator0
    "Sprite Renderer "=DataInterface() {propertyValues {SpriteRenderer=&self:Coin_0.NiagaraSpriteRendererProperties_0}}
    SubUV Animation Mode:ENiagara_SubUVLookupModeV2=::NewEnumerator2
    UseEndFrame=false
    UseStartFrame=false
   )
   &ScaleColor(
    Scale RGB=&VectorFromFloat(Value=&RemapRange(Input Value To Remap=$Particles.SineWave, Input Min=-1, Input Max=1, Output Min=1, Output Max=3))
    ScaleA=true
    ScaleRGB=true
   )
  ]
  rendererClasses [&NiagaraMeshRendererProperties, &NiagaraSpriteRendererProperties]
  renderers [
   {rendererClass=&NiagaraMeshRendererProperties, rendererIndex=0}
   {rendererClass=&NiagaraSpriteRendererProperties, rendererIndex=1}
  ]
 }
 Default_Emote_TraversalSmoke_Loop {
  emitterUpdateScript [
   &EmitterState()
   &SpawnPerUnit(Module.Use Max Movement Threshold=true, Module.Use Movement Tolerance=false, Spawn Spacing=15, Use Spawn Probability=false)
   &SpawnRate(SpawnRate=&FN_FrontEndFloatModifier(Gameplay=0, Front End=20), Use Spawn Probability=false)
   &SpawnPerUnit(
    Module.Use Max Movement Threshold=true
    Module.Use Movement Tolerance=true
    Movement Tolerance=3
    Spawn Spacing:NiagaraFloat=&FN_PickFloatDependOnPlatform(
     TargetPlatform=DataInterface() {propertyValues {Platforms {qualityLevelMask=-8}}}
     TargetPlatformActive=20
     TargetPlatformNoActive=&FN_PickFloatDependOnPlatform(
      TargetPlatform=DataInterface() {propertyValues {Platforms {qualityLevelMask=-28}}}
      TargetPlatformActive=8
      TargetPlatformNoActive=&FN_PickFloatDependOnPlatform(
       TargetPlatform=DataInterface() {propertyValues {Platforms {qualityLevelMask=-30}}}
       TargetPlatformActive=14
       TargetPlatformNoActive=&FN_PickFloatDependOnPlatform(TargetPlatform=DataInterface() {propertyValues {Platforms {qualityLevelMask=-31}}}, TargetPlatformActive=50, TargetPlatformNoActive=0)
      )
     )
    )
    Use Spawn Probability=false
   ) {enabled=false, module=SpawnPerUnit001}
  ]
  particleSpawnScript [
   &PartitionParticles()
   &InitializeParticle(
    Color Channel Mode:ENiagaraChannelCorrelation=::NewEnumerator0
    Color Maximum=<rgba 0.781 0.7271 0.5498 0.4>
    Color Minimum=<rgba 0.478 0.3791 0.2199 0.2>
    Color Mode=::NewEnumerator2
    Lifetime Max=0.8
    Lifetime Min=0.5
    Lifetime Mode=::NewEnumerator1
    Sprite Rotation Mode=::NewEnumerator2
    Sprite Size Mode=::NewEnumerator1
    Uniform Sprite Size Max=25
    Uniform Sprite Size Min=15
    UsePositionOffset=false
    Write Material Random=true
   ) {module=InitializeParticle001}
   &SkeletalMeshLocation(
    Bone / Socket Index=$Particles.PartitionParticles.Partition
    Bone Sampling Mode=::NewEnumerator4
    MaskSpawnGroup=false
    Orientation Sampling=::NewEnumerator1
    Sampled Position Offset=&FN_FrontEndVector3Modifier(Gameplay=<0 0 0>, Front End=<0 -7 0>)
    Skeletal Mesh=$System.WheelLocations
   ) {enabled=false}
   &SkeletalMeshLocation(
    Bone / Socket Index=&MakeCustomIntFromBool(
     Bool=$User.bIsFrontEndPreview
     True INT=&SelectIntFromArray(Direct Array Index=$Particles.PartitionParticles.Partition, Int Selection Array=DataInterface() {propertyValues {IntData [2,3]}})
     False INT=&SelectIntFromArray(
      Direct Array Index=$Particles.PartitionParticles.Partition
      Direct Array Mode=::NewEnumerator1
      Int Selection Array=DataInterface() {propertyValues {IntData [0,1]}}
     )
    )
    Bone Sampling Mode=::NewEnumerator4
    MaskSpawnGroup=false
    Orientation Sampling=::NewEnumerator1
    Sampled Position Offset=&FN_FrontEndVector3Modifier(Gameplay=<0 0 0>, Front End=<0 -7 0>)
    Skeletal Mesh=$System.WheelLocations
   ) {module=SkeletalMeshLocation001}
   &InheritVelocity(Inherited Velocity Speed Limit=1000, Speed Limit=true, UseSpeedThreshold=false) {module=InheritVelocity001}
   &AddVelocity(
    Velocity=&FN_FrontEndVector3Modifier(Gameplay=<0 -2 0.3>, Front End=<0 -4 0>)
    Velocity Speed Scale=&RandomRangeFloat(Minimum=0, Maximum=100)
   ) {module=AddVelocity001}
   &AccelerationForce(Acceleration=&UniformRangedVector(Minimum=<0 -100 0>, Maximum=<0 -250 0>), Coordinate Space=::Local)
   &ApplyInitialForces()
  ]
  particleUpdateScript [
   &ParticleState()
   &Drag(Drag=10, Use Linear Drag=true, Use Rotational Drag=true)
   &SolveForcesAndVelocity(Clamp Velocity=false, Limit Acceleration=false)
   &ScaleSpriteSize(Uniform Curve Sprite Scale=DataInterface() {propertyValues {Curve {keys [{arriveTangent=1.491, interpMode=RCIM_Cubic, leaveTangent=1.491, tangentMode=RCTM_User, time=0, value=0.5}, <at 1 3 RCIM_Cubic>]}}})
   &DynamicMaterialParameters(
    Index 0 Param 1=&Add_Float(
     A=&RandomRangeFloat(Minimum=0, Maximum=1)
     B=&FloatFromCurve(
      FloatCurve=DataInterface() {propertyValues {Curve {keys [
          <at 0 0 RCIM_Cubic>
          {arriveTangent=1.5947, interpMode=RCIM_Cubic, leaveTangent=1.5947, tangentMode=RCTM_User, time=0.2557, value=0.5257}
          <at 1 1 RCIM_Cubic>
         ]}}}
      CurveIndex=$Particles.Age
      Scale Curve=0.25
     )
    )
    Module.Index0Float0Write=true
    Module.Index0Float1Write=true
    Module.Index0Float2Write=true
    Module.Index0Float3Write=true
   ) {module=DynamicMaterialParameters001}
   &CameraOffset(Camera Offset Amount=15)
   &ScaleColor(
    Scale Mode=::NewEnumerator1
    Scale RGBA=&MakeVector4_FromLinearColor(LinearColor=&FNTimeOfDayColor(
      Night Color=&RandomRangeLinearColor(
       Minimum=&ColorFromCurve(ColorCurve=DataInterface() {propertyValues {
          AlphaCurve {keys [<at 0 1 RCIM_Cubic>]}
          BlueCurve {keys [<at 0 0.5411 RCIM_Cubic>,<at 1 0.2515 RCIM_Cubic>]}
          GreenCurve {keys [<at 0 0.3331 RCIM_Cubic>,<at 1 0.7243 RCIM_Cubic>]}
          RedCurve {keys [<at 0 0.2153 RCIM_Cubic>,<at 1 0.8744 RCIM_Cubic>]}
         }})
       Maximum=&ColorFromCurve(ColorCurve=DataInterface() {propertyValues {
          AlphaCurve {keys [<at 0 1 RCIM_Cubic>]}
          BlueCurve {keys [<at 0 1 RCIM_Cubic>,<at 1 0.2422 RCIM_Cubic>]}
          GreenCurve {keys [<at 0 0.6471 RCIM_Cubic>,<at 1 0.708 RCIM_Cubic>]}
          RedCurve {keys [<at 0 0.4118 RCIM_Cubic>,<at 1 0.8559 RCIM_Cubic>]}
         }})
      )
      Morning Color=&RandomRangeLinearColor(Minimum=<rgba 0.3361 0.278 0.5712 0.6>, Maximum=<rgba 0.3377 0.5872 0.9253 0.8>)
      Day Color=&RandomRangeLinearColor(Minimum=<rgba 0.478 0.4265 0.4121 0.8>, Maximum=<rgba 0.781 0.718 0.6943 0.9>)
      Evening Color=&RandomRangeLinearColor(Minimum=<rgba 0.478 0.478 0.478 1>, Maximum=<rgba 0.781 0.781 0.781 1>)
     ))
    ScaleRGBA=true
   ) {module=ScaleColor001}
   &ScaleColor(
    Scale Alpha=&Multiply_Float(
     A=&FloatFromCurve(
      FloatCurve=DataInterface() {propertyValues {Curve {keys [
          {interpMode=RCIM_Cubic, tangentMode=RCTM_User, time=0, value=0}
          <at 0.3002 1 RCIM_Cubic>
          {arriveTangent=-2.453, tangentMode=RCTM_User, time=0.9, value=0}
         ]}}}
      Scale Curve=&FN_FrontEndFloatModifier(Gameplay=0.75, Front End=1)
     )
     B=&FloatFromCurve(
      FloatCurve=DataInterface() {propertyValues {Curve {keys [
          {arriveTangent=0.8395, interpMode=RCIM_Cubic, leaveTangent=0.8395, tangentMode=RCTM_User, time=0, value=0.7}
          {interpMode=RCIM_Cubic, tangentMode=RCTM_User, time=0.5, value=1}
          {arriveTangent=-0.9582, interpMode=RCIM_Cubic, leaveTangent=-0.9582, tangentMode=RCTM_User, time=1, value=0.7}
         ]}}}
      CurveIndex=Linked() {linkedVariable {name=NPC.FortniteNPC.TimeOfDay_0-1, type=NiagaraFloat}}
     )
    )
    ScaleA=true
    ScaleRGB=false
   )
  ]
  rendererClasses [&NiagaraSpriteRendererProperties]
  renderers [
   {rendererClass=&NiagaraSpriteRendererProperties}
   {rendererClass=&NiagaraSpriteRendererProperties, rendererIndex=1}
  ]
 }
 Glow {
  emitterUpdateScript [
   &EmitterState(Life Cycle Mode=::NewEnumerator1, Loop Behavior=::NewEnumerator1, Loop Duration Mode=::NewEnumerator1, UseLoopDelay=false)
   &SpawnBurst_Instantaneous(Use Spawn Probability=false, UseLoopCountLimit=false)
  ]
  particleSpawnScript [
   &InitializeParticle(
    Color=<rgba 1 0.7561 0.3167 0.1>
    Lifetime=0.25
    Position Offset=<-5 5 0>
    Sprite Size Mode=::NewEnumerator3
    Uniform Sprite Size=83.3793
    UsePositionOffset=true
    Write Material Random=true
   )
   &CameraOffset(Camera Offset Amount=10)
  ]
  particleUpdateScript [
   &ParticleState(Kill Particles When Lifetime Has Elapsed=false, Let Infinitely Lived Particles Die When Emitter Deactivates=true)
   &FN_Light_Attributes(
    Color=<rgba 1 0.8075 0.3503 1>
    Create Color Attribute=true
    Create Exponent Attribute=true
    Create Radius Attribute=true
    Enable Particle Light Enable=true
    Exponent=2
    Light Enabled=true
    Radius=100
    Scale Color=&FloatFromCurve(
     FloatCurve=DataInterface() {propertyValues {Curve {keys [
         <at 0 0 RCIM_Cubic>
         {arriveTangentWeight=0.0581, interpMode=RCIM_Cubic, leaveTangentWeight=0.349, tangentMode=RCTM_User, time=0.0705, value=1}
         {arriveTangent=-0.0405, interpMode=RCIM_Cubic, leaveTangent=-0.0405, tangentMode=RCTM_User, time=2.2171, value=1}
        ]}}}
     CurveIndex=$Particles.Age
     Scale Curve=5
    )
   )
   /SetVariables(Particles.LightPosition:NiagaraPosition=&AddVectorToPosition(Position=<0 -40 0>, Vector=<0 0 0>)) {isSetParameters=true}
   &ScaleColor(
    Scale Alpha=&FloatFromCurve(FloatCurve=DataInterface() {propertyValues {Curve {keys [{interpMode=RCIM_Cubic, tangentMode=RCTM_User, time=0, value=0}, {interpMode=RCIM_Cubic, tangentMode=RCTM_User, time=1, value=1}]}}})
    ScaleA=true
    ScaleRGB=true
   )
  ]
  rendererClasses [&NiagaraSpriteRendererProperties, &NiagaraLightRendererProperties]
  renderers [
   {rendererClass=&NiagaraSpriteRendererProperties}
   {rendererClass=&NiagaraLightRendererProperties, rendererIndex=1}
  ]
 }
 Light_Rays001 {
  emitterUpdateScript [
   &EmitterState(
    Inactive Response=::NewEnumerator1
    Life Cycle Mode=::NewEnumerator1
    Loop Behavior=::NewEnumerator1
    Loop Delay=2.7
    Loop Duration Mode=::NewEnumerator1
    UseLoopDelay=false
   )
   &SpawnBurst_Instantaneous(Spawn Count=1, Use Spawn Probability=false)
  ]
  particleSpawnScript [
   &InitializeParticle(
    Color=<rgba 1 0.7016 0.2711 0.7043>
    Lifetime=0.25
    Mesh Scale=&Multiply_VectorByFloat(Vector=<1 1 1>, Float=0.97)
    Mesh Scale Mode=::NewEnumerator4
    Position Offset=<-8.9728 -11.3989 -5.0383>
    UsePositionOffset=true
    Write Material Random=true
   ) {module=InitializeParticle001}
   &CameraOffset() {module=CameraOffset001}
   &InitialMeshOrientation(Mesh Axis To Orient=<0 1 0>, Mesh Orientation Mode=::NewEnumerator2, Orientation Vector=<0 1 0>, Use Rotation Vector=true)
  ]
  particleUpdateScript [
   &ParticleState(Kill Particles When Lifetime Has Elapsed=false, Let Infinitely Lived Particles Die When Emitter Deactivates=true)
   &DynamicMaterialParameters(
    Index 0 Param 1=&FloatFromCurve(
     FloatCurve=DataInterface() {propertyValues {Curve {keys [
         {arriveTangent=0.9332, interpMode=RCIM_Cubic, leaveTangent=0.9332, tangentMode=RCTM_User, time=0, value=0}
         {arriveTangent=0.414, interpMode=RCIM_Cubic, leaveTangent=0.414, time=0.6, value=0.6}
         {interpMode=RCIM_Cubic, tangentMode=RCTM_User, time=1, value=0.414}
        ]}}}
     Scale Curve=&Multiply_Float(
      A=&RandomRangeFloat(Minimum=10, Maximum=20)
      B=&Lerp_Float(A=0, B=4, Alpha=0)
     )
    )
    Index 0 Param 2=&FloatFromCurve(
     FloatCurve=DataInterface() {propertyValues {Curve {keys [
         {arriveTangent=-1.369, interpMode=RCIM_Cubic, leaveTangent=-1.369, tangentMode=RCTM_User, time=0, value=1}
         {interpMode=RCIM_Cubic, tangentMode=RCTM_User, time=1, value=0.6}
        ]}}}
     Scale Curve=&RandomRangeFloat(Minimum=0.5, Maximum=1.5)
    )
    Module.Index0Float0Write=true
    Module.Index0Float1Write=true
    Module.Index0Float2Write=true
    Module.Index0Float3Write=true
   )
   &ScaleColor(
    Scale Alpha=&FloatFromCurve(
     FloatCurve=DataInterface() {propertyValues {Curve {keys [{interpMode=RCIM_Cubic, tangentMode=RCTM_User, time=0, value=0}, {interpMode=RCIM_Cubic, tangentMode=RCTM_User, time=1, value=1}]}}}
     Scale Curve=0.5
    )
    ScaleA=true
    ScaleRGB=true
   )
  ]
  rendererClasses [&NiagaraMeshRendererProperties]
  renderers [{rendererClass=&NiagaraMeshRendererProperties, rendererIndex=0}]
 }
 Lights {
  emitterUpdateScript [
   &EmitterState(
    Inactive Response=::NewEnumerator1
    Life Cycle Mode=::NewEnumerator1
    Loop Behavior=::NewEnumerator1
    Loop Duration Mode=::NewEnumerator1
    UseLoopDelay=false
   )
   &SpawnBurst_Instantaneous(
    Spawn Count=&MakeIntFromFloat(Float=&FN_PickFloatDependOnPlatform(TargetPlatform=DataInterface() {propertyValues {Platforms {qualityLevelMask=-8}}}, TargetPlatformActive=4, TargetPlatformNoActive=2))
    Use Spawn Probability=false
    UseLoopCountLimit=false
   )
  ]
  particleSpawnScript [
   &PartitionParticles(Number Of Partitions=&MakeIntFromFloat(Float=&FN_PickFloatDependOnPlatform(TargetPlatform=DataInterface() {propertyValues {Platforms {qualityLevelMask=-8}}}, TargetPlatformActive=4, TargetPlatformNoActive=2)))
   &InitializeParticle(
    Color=&MaskLinearColor(
     Mask Parameter=$Particles.PartitionParticles.Partition
     Mask=0
     Masked Linear Color=<rgba 1 0.7996 0.3164 0.3024>
     Passthrough Linear Color=&MaskLinearColor(
      Mask Parameter=$Particles.PartitionParticles.Partition
      Mask=1
      Masked Linear Color=<rgba 1 0.7996 0.3164 0.3024>
      Passthrough Linear Color=<rgba 1 0.0833 0.0907 0.293>
     )
    )
    Lifetime=0.6
    Sprite Size Mode=::NewEnumerator3
    Uniform Sprite Size=15.035
    UsePositionOffset=false
    Write Material Random=true
   )
   &CameraOffset(Camera Offset Amount=10)
   &SkeletalMeshLocation(
    Bone / Socket Index=$Particles.PartitionParticles.Partition
    Bone / Socket Sampling Mode=::NewEnumerator5
    MaskSpawnGroup=false
    Mesh Sampling Type=::NewEnumerator6
    Orientation Sampling=::NewEnumerator1
    Skeletal Mesh=DataInterface() {propertyValues {
      DefaultMesh=/BRCosmetics/Animation/Props/Gadgets_GetawayCar/Meshes/GetawayCar_Gadget
      ExcludeBoneName=root
      OptionalComponentClass=null
      PreviewMesh=&GetawayCar_Gadget
      SkinningMode=SkinOnTheFly
      SoftSourceActor=null
      SourceMode=Default
      UvSetIndex=0
      WholeMeshLOD=-1
      bExcludeBone=false
      bReadDeformedGeometry=true
      bRequireCurrentFrameData=false
      ComponentTags []
      FilteredBones []
      FilteredSockets [HeadlightLeftFX, HeadlightRightFX, LeftLightFX, RightLightFX]
      MeshUserParameter {parameter {name=None, typeDefHandle {registeredTypeIndex=1}, varData []}}
      SamplingRegions []
     }}
   )
   /SetVariables(Particles.Position=&AddVectorToPosition(
     Position=$Particles.SkeletalMeshLocation.SampledPosition
     Vector=&MaskVector(
      Mask Parameter=$Particles.PartitionParticles.Partition
      Mask=0
      Masked Vector=<-5 -5 0>
      Passthrough Vector=&MaskVector(
       Mask Parameter=$Particles.PartitionParticles.Partition
       Mask=1
       Masked Vector=<5 -5 0>
       Passthrough Vector=&MaskVector(Mask Parameter=$Particles.PartitionParticles.Partition, Mask=2, Masked Vector=<0 0 0>, Passthrough Vector=<0 0 0>)
      )
     )
    )) {isSetParameters=true}
  ]
  particleUpdateScript [
   &ParticleState(Kill Particles When Lifetime Has Elapsed=false, Let Infinitely Lived Particles Die When Emitter Deactivates=true)
   &ScaleColor(
    Scale Alpha=&FloatFromCurve(
     FloatCurve=DataInterface() {propertyValues {Curve {keys [
         {arriveTangent=4.1785, interpMode=RCIM_Cubic, leaveTangent=4.1785, tangentMode=RCTM_User, time=0, value=0}
         {arriveTangent=1.1179, interpMode=RCIM_Cubic, leaveTangent=1.1179, tangentMode=RCTM_User, time=0.1822, value=0.6751}
         {arriveTangent=0.0224, interpMode=RCIM_Cubic, leaveTangent=0.0224, tangentMode=RCTM_User, time=0.5068, value=1}
        ]}}}
     CurveIndex=$Emitter.Age
    )
    ScaleA=true
    ScaleRGB=true
   )
   &FN_Light_Attributes(
    Color=&MaskLinearColor(
     Mask Parameter=$Particles.PartitionParticles.Partition
     Mask=0
     Masked Linear Color=<rgba 1 0.7996 0.3164 0.5>
     Passthrough Linear Color=&MaskLinearColor(
      Mask Parameter=$Particles.PartitionParticles.Partition
      Mask=1
      Masked Linear Color=<rgba 1 0.7996 0.3164 0.5>
      Passthrough Linear Color=<rgba 2 0 0 1>
     )
    )
    Create Color Attribute=true
    Create Exponent Attribute=true
    Create Position Attribute=true
    Create Radius Attribute=true
    Enable Particle Light Enable=true
    Exponent=2
    Light Enabled=true
    Position=&AddVectorToPosition(
     Position=$Particles.SkeletalMeshLocation.SampledPosition
     Vector=&MaskVector(
      Mask Parameter=$Particles.PartitionParticles.Partition
      Mask=0
      Masked Vector=<-2 25 -20>
      Passthrough Vector=&MaskVector(
       Mask Parameter=$Particles.PartitionParticles.Partition
       Mask=1
       Masked Vector=<2 25 -20>
       Passthrough Vector=&MaskVector(Mask Parameter=$Particles.PartitionParticles.Partition, Mask=2, Masked Vector=<0 -12 0>, Passthrough Vector=<0 -12 0>)
      )
     )
    )
    Radius=&MaskFloat(
     Mask Parameter=$Particles.PartitionParticles.Partition
     Mask=0
     Masked Float=70
     Passthrough Float=&MaskFloat(Mask Parameter=$Particles.PartitionParticles.Partition, Mask=1, Masked Float=70, Passthrough Float=20)
    )
    Scale Color=&FloatFromCurve(
     FloatCurve=DataInterface() {propertyValues {Curve {keys [
         {arriveTangent=4.1785, interpMode=RCIM_Cubic, leaveTangent=4.1785, tangentMode=RCTM_User, time=0, value=0}
         {arriveTangent=1.1179, interpMode=RCIM_Cubic, leaveTangent=1.1179, tangentMode=RCTM_User, time=0.1822, value=0.6751}
         {arriveTangent=0.0224, interpMode=RCIM_Cubic, leaveTangent=0.0224, tangentMode=RCTM_User, time=0.5068, value=1}
        ]}}}
     CurveIndex=$Emitter.Age
     Scale Curve=&MaskFloat(
      Mask Parameter=$Particles.PartitionParticles.Partition
      Mask=0
      Masked Float=2
      Passthrough Float=&MaskFloat(Mask Parameter=$Particles.PartitionParticles.Partition, Mask=1, Masked Float=2, Passthrough Float=2)
     )
    )
   )
  ]
  rendererClasses [&NiagaraSpriteRendererProperties, &NiagaraLightRendererProperties]
  renderers [
   {rendererClass=&NiagaraSpriteRendererProperties}
   {rendererClass=&NiagaraLightRendererProperties, rendererIndex=1}
  ]
 }
 NE_Topper_Bling {
  emitterUpdateScript [
   &EmitterState(Life Cycle Mode=::NewEnumerator1, Loop Behavior=::NewEnumerator1, Loop Duration Mode=::NewEnumerator1, UseLoopDelay=false)
   &SpawnRate(Spawn Probability=0.8, SpawnRate=3, Use Spawn Probability=true)
  ]
  particleSpawnScript [
   &InitializeParticle(
    Color=<rgba 5 4.0621 1.9693 1>
    Lifetime Max=0.5
    Lifetime Min=0.3
    Lifetime Mode=::NewEnumerator1
    Sprite Rotation Mode=::NewEnumerator2
    Sprite Size Mode=::NewEnumerator1
    Sprite UV Mode=::NewEnumerator4
    Uniform Sprite Size Min=20
    UsePositionOffset=false
    Write Material Random=true
   ) {module=InitializeParticle001}
   &SkeletalMeshLocation(Bone Sampling Mode=::NewEnumerator4, MaskSpawnGroup=false, Skeletal Mesh=$System.VehicleLocation)
   &ShapeLocation(
    Hemisphere Distribution=<0.5 1>
    MaskSpawnGroup=false
    Offset=<0 0 5>
    Offset Mode=::NewEnumerator0
    Override Seed=false
    Rotation Mode=::NewEnumerator2
    Shape Origin=$Particles.Position
    Sphere Radius=15
    Sphere Surface Distribution=1
    UseOwnerScale=true
    Yaw / Pitch / Roll=<0 0 -0.4914>
   )
   &CameraOffset(Camera Offset Amount=-15)
  ]
  particleUpdateScript [
   &ParticleState()
   &SpriteRotationRate(Rotation Rate=300)
   &ScaleColor(
    Scale Alpha=&FloatFromCurve(FloatCurve=DataInterface() {propertyValues {Curve {keys [<at -0.0092 0.0144>,<at 0.0967 0.9767>,<at 1.0011 0.002>]}}})
    ScaleA=true
    ScaleRGB=true
   )
   &ScaleSpriteSize(Scale Factor=&Vector2DFromFloat(Value=&FloatFromCurve(FloatCurve=DataInterface() {propertyValues {Curve {keys [
          <at -0.0046 0 RCIM_Cubic>
          {
           arriveTangent=0.0048
           arriveTangentWeight=0.1184
           interpMode=RCIM_Cubic
           leaveTangent=0.0054
           leaveTangentWeight=0.5412
           tangentMode=RCTM_Break
           tangentWeightMode=RCTWM_WeightedBoth
           time=0.1214
           value=1
          }
          <at 1.0011 0.002 RCIM_Cubic>
         ]}}})))
  ]
  rendererClasses [&NiagaraSpriteRendererProperties]
  renderers [{rendererClass=&NiagaraSpriteRendererProperties}]
 }
}
summary {
 systemName=NS_Emote_GetawayCar_Money
 emitters [
  {emitterName=Default_Emote_TraversalSmoke_Loop, rendererClasses [&NiagaraSpriteRendererProperties]}
  {emitterName=Coin, rendererClasses [&NiagaraMeshRendererProperties, &NiagaraSpriteRendererProperties]}
  {emitterName=Glow, rendererClasses [&NiagaraSpriteRendererProperties, &NiagaraLightRendererProperties]}
  {emitterName=Light_Rays001, rendererClasses [&NiagaraMeshRendererProperties]}
  {emitterName=NE_Topper_Bling, rendererClasses [&NiagaraSpriteRendererProperties]}
  {emitterName=Lights, rendererClasses [&NiagaraSpriteRendererProperties, &NiagaraLightRendererProperties]}
 ]
 userVariables [
  User.bIsFrontEnd {type=NiagaraBool}
  User.bIsFrontEndPreview {type=NiagaraBool}
  User.bIsInWater {type=NiagaraBool}
 ]
}
