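# Niagara system for the police Getaway Car emote, adding wheel dust and debris, flashing red/blue lights, and warm and red headlight/taillight glows.
# Smoke and debris spawn by movement at wheel locations, with debris suppressed in water and smoke adapting to front end, platform and time of day.
# Two-particle light emitters follow skeletal sockets, using partition masks for color and blink, plus camera-facing fades.
# The flashing pair also drives light renderers.
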NiagaraSystem /BRCosmetics/Animation/Game/MainPlayer/Emotes/GetawayCar/FX/NS_Emote_GetawayCar_Police
dependencies {
 usedDataInterfaces [
  &NiagaraDataInterfacePlatformSet
  &NiagaraDataInterfaceCurve
  &NiagaraDataInterfaceColorCurve
  &NiagaraDataInterfaceCamera
  &NiagaraDataInterfaceSkeletalMesh
 ]
 usedDynamicInputs [&SimulationPosition]
 usedModules [
  &self:SystemSpawnScript.SystemScriptSource.SystemScriptGraph.NiagaraNodeAssignment_4.SetVariables_#1
  &self:SystemSpawnScript.SystemScriptSource.SystemScriptGraph.NiagaraNodeAssignment_2.SetVariables_#2
  &self:SystemSpawnScript.SystemScriptSource.SystemScriptGraph.NiagaraNodeAssignment_0.SetVariables_#3
  &self:SystemSpawnScript.SystemScriptSource.SystemScriptGraph.NiagaraNodeAssignment_1.SetVariables_#4
  &SystemState
 ]
}
emitters {
 Debris {
  emitterUpdateScript [
   &EmitterState(Enable Distance Culling=true, MaxDistance=1150, Module.bUseMaxDistance=true, Module.bUseMinDistance=false, Scalability Mode=::NewEnumerator1)
   &SpawnRate(SpawnRate=20, Use Spawn Probability=false) {enabled=false}
   &SpawnPerUnit(
    Module.Use Max Movement Threshold=true
    Module.Use Movement Tolerance=true
    Movement Tolerance=1
    Spawn Per Unit:NiagaraFloat=&Multiply_Float(A=&MakeCustomFloatFromBool(Bool=$User.bIsInWater, True Float=0, False Float=1), B=20)
    Use Spawn Probability=false
   )
  ]
  particleSpawnScript [
   &InitializeParticle(
    Color=&ColorFromCurve(
     ColorCurve=DataInterface() {propertyValues {
       AlphaCurve {keys [<at -0.0023 1 RCIM_Cubic>]}
       BlueCurve {keys [<at -0.0023 0.2122 RCIM_Cubic>,<at 0.1383 0.2122>,<at 0.2178 0.1859>,<at 0.4486 0.75>,<at 0.7126 0.1859>,<at 1.0234 0.8802>]}
       GreenCurve {keys [<at -0.0023 0.2325 RCIM_Cubic>,<at 0.1383 0.2325>,<at 0.2178 0.2005>,<at 0.4486 0.8456>,<at 0.7126 0.2122>,<at 1.0234 0.9474>]}
       RedCurve {keys [<at -0.0023 0.25 RCIM_Cubic>,<at 0.1383 0.25>,<at 0.2178 0.2188>,<at 0.4486 1>,<at 0.7126 0.2188>,<at 1.0234 1>]}
      }}
     CurveIndex=&UniformRangedFloat(Minimum=0, Maximum=1)
    )
    Lifetime=&UniformRangedFloat(Minimum=0.25, Maximum=0.75)
    Sprite Rotation Mode=::NewEnumerator2
    Sprite Size Mode=::NewEnumerator3
    Uniform Sprite Size=&MakeFloatFromVector2d(Vector 2D=&Vector2DFromFloat(Value=&UniformRangedFloat(Minimum=4, Maximum=10)))
    UsePositionOffset=false
    Write Material Random=true
   ) {module=InitializeParticle001}
   &SkeletalMeshLocation(
    Bone Sampling Mode=::NewEnumerator4
    MaskSpawnGroup=false
    Sampled Position Offset=<0 -48 0>
    Sampled Position Offset Coordinate Space:ENiagaraSkelMeshTransforms=::NewEnumerator2
    Skeletal Mesh=$System.WheelLocations
   )
   /Niagara/Modules/Spawn/Location/SphereLocation(MaskSpawnGroup=false, Module.Override Randomness=false, Non Uniform Scale=<3 1 1>, Sphere Radius=10, Surface Only=false)
   &AddVelocity(Velocity=&RandomRangeVector(Minimum=<0 -20 0>, Maximum=<1 20 25>))
   &InheritVelocity(Module.Limit Speed=true, Velocity Scale=&UniformRangedVector(Minimum=<0 -5 0>, Maximum=<0 -5 0>))
  ]
  particleUpdateScript [
   &ParticleState()
   &Drag(Use Linear Drag=true, Use Rotational Drag=false)
   &GravityForce(Gravity=<0 0 -500>)
   &Collision(
    Age Colliding Particles=false
    Collision Enabled=&IsPlatformSetActive(New Platform Set=DataInterface() {propertyValues {Platforms {qualityLevelMask=-8}}})
    KillOnCollision=false
    Module.Control Rest State Via Surface Normal=false
    Module.EnableMaxCollisionCount=false
    Module.Randomize Collision Normal Vector=false
    Radius Calculation Type=::NewEnumerator1
   )
   &Collision(
    Age Colliding Particles=false
    Analytical Collision Plane Position 1:Vector3f=<0 0 -20>
    Analytical Collision Plane Space:ENiagaraCoordinateSpace=::Local
    CPU Collision Type=::NewEnumerator3
    Collision Enabled=&IsPlatformSetActive(New Platform Set=DataInterface() {propertyValues {Platforms {qualityLevelMask=-28}}})
    KillOnCollision=false
    Module.Control Rest State Via Surface Normal=false
    Module.EnableMaxCollisionCount=false
    Module.Randomize Collision Normal Vector=false
    Perform 2 Analytical Collision Queries=false
    Restitution=0
   ) {module=Collision001}
   &ScaleColor(Scale RGB=<1.2 1.2 1.2>, ScaleA=true, ScaleRGB=true)
   &ScaleSpriteSize(Scale Factor=&Vector2DFromFloat(Value=&FloatFromCurve(FloatCurve=DataInterface() {propertyValues {Curve {keys [
          <at 0.7 0.9701>
          {arriveTangent=-3.2149, interpMode=RCIM_Cubic, leaveTangent=-3.2149, tangentMode=RCTM_User, time=0.9205, value=0.9191}
          <at 1.0011 0.002>
         ]}}})))
   &SpriteRotationRate()
   &SolveForcesAndVelocity(Clamp Velocity=false, Limit Acceleration=false)
   &ScaleSpriteSizeBySpeed(
    Max Scale Factor=&Vector2DFromFloat(Value=1)
    Min Scale Factor=&Vector2DFromFloat(Value=0.25)
    Velocity Threshold=200
   )
  ]
  rendererClasses [&NiagaraSpriteRendererProperties]
  renderers [{rendererClass=&NiagaraSpriteRendererProperties}]
 }
 Default_Emote_TraversalSmoke_Loop {
  emitterUpdateScript [
   &EmitterState()
   &SpawnRate(SpawnRate=&FN_FrontEndFloatModifier(Gameplay=0, Front End=20), Use Spawn Probability=false)
   &SpawnPerUnit(
    Module.Use Max Movement Threshold=true
    Module.Use Movement Tolerance=true
    Movement Tolerance=3
    Spawn Spacing:NiagaraFloat=&FN_PickFloatDependOnPlatform(
     TargetPlatform=DataInterface() {propertyValues {Platforms {qualityLevelMask=-8}}}
     TargetPlatformActive=5
     TargetPlatformNoActive=&FN_PickFloatDependOnPlatform(
      TargetPlatform=DataInterface() {propertyValues {Platforms {qualityLevelMask=-28}}}
      TargetPlatformActive=8
      TargetPlatformNoActive=&FN_PickFloatDependOnPlatform(
       TargetPlatform=DataInterface() {propertyValues {Platforms {qualityLevelMask=-30}}}
       TargetPlatformActive=14
       TargetPlatformNoActive=&FN_PickFloatDependOnPlatform(TargetPlatform=DataInterface() {propertyValues {Platforms {qualityLevelMask=-31}}}, TargetPlatformActive=14, TargetPlatformNoActive=0)
      )
     )
    )
    Use Spawn Probability=false
   ) {module=SpawnPerUnit001}
  ]
  particleSpawnScript [
   &InitializeParticle(
    Color Channel Mode:ENiagaraChannelCorrelation=::NewEnumerator0
    Color Maximum=<rgba 0.781 0.7271 0.5498 0.5>
    Color Minimum=<rgba 0.478 0.3791 0.2199 0.03>
    Color Mode=::NewEnumerator2
    Lifetime Max=0.4
    Lifetime Min=0.5
    Lifetime Mode=::NewEnumerator1
    Sprite Rotation Mode=::NewEnumerator2
    Sprite Size Mode=::NewEnumerator1
    Uniform Sprite Size Max=25
    Uniform Sprite Size Min=15
    UsePositionOffset=false
    Write Material Random=true
   ) {module=InitializeParticle001}
   &SkeletalMeshLocation(
    Bone Sampling Mode=::NewEnumerator4
    MaskSpawnGroup=false
    Sampled Position Offset=&FN_FrontEndVector3Modifier(Gameplay=<0 -48 0>, Front End=<0 -35 -10>)
    Skeletal Mesh=$System.WheelLocations
   )
   &InheritVelocity(Inherited Velocity Speed Limit=1000, Speed Limit=true, UseSpeedThreshold=false) {module=InheritVelocity001}
   &AddVelocity(
    Velocity=<0 -2 1>
    Velocity Speed Scale=&FN_FrontEndFloatModifier(
     Gameplay=&RandomRangeFloat(Minimum=0, Maximum=60)
     Front End=&RandomRangeFloat(Minimum=20, Maximum=150)
    )
   ) {module=AddVelocity001}
   &AccelerationForce(Acceleration=&UniformRangedVector(Minimum=<0 -100 0>, Maximum=<0 -250 0>), Coordinate Space=::Local)
   &ApplyInitialForces()
  ]
  particleUpdateScript [
   &ParticleState()
   &Drag(Drag=10, Use Linear Drag=true, Use Rotational Drag=true)
   &SolveForcesAndVelocity(Clamp Velocity=false, Limit Acceleration=false)
   &ScaleSpriteSize(Uniform Curve Sprite Scale=DataInterface() {propertyValues {Curve {keys [{arriveTangent=1.491, interpMode=RCIM_Cubic, leaveTangent=1.491, tangentMode=RCTM_User, time=0, value=0.5}, <at 1 3 RCIM_Cubic>]}}})
   &DynamicMaterialParameters(
    Index 0 Param 1=&Add_Float(
     A=&RandomRangeFloat(Minimum=0, Maximum=1)
     B=&FloatFromCurve(
      FloatCurve=DataInterface() {propertyValues {Curve {keys [
          <at 0 0 RCIM_Cubic>
          {arriveTangent=1.5947, interpMode=RCIM_Cubic, leaveTangent=1.5947, tangentMode=RCTM_User, time=0.2557, value=0.5257}
          <at 1 1 RCIM_Cubic>
         ]}}}
      CurveIndex=$Particles.Age
      Scale Curve=0.25
     )
    )
    Module.Index0Float0Write=true
    Module.Index0Float1Write=true
    Module.Index0Float2Write=true
    Module.Index0Float3Write=true
   ) {module=DynamicMaterialParameters001}
   &CameraOffset(Camera Offset Amount=15)
   &ScaleColor(
    Scale Mode=::NewEnumerator1
    Scale RGBA=&MakeVector4_FromLinearColor(LinearColor=&FNTimeOfDayColor(
      Night Color=&RandomRangeLinearColor(Minimum=<rgba 0.1294 0.3996 0.8698 0.9>, Maximum=<rgba 0.4126 0.6462 1 1>)
      Morning Color=&RandomRangeLinearColor(Minimum=<rgba 0.2813 0.2326 0.478 0.8589>, Maximum=<rgba 0.3377 0.5872 0.9253 0.913>)
      Day Color=&RandomRangeLinearColor(Minimum=<rgba 0.478 0.4265 0.4121 0.5>, Maximum=<rgba 0.781 0.718 0.6943 0.7>)
      Evening Color=&RandomRangeLinearColor(Minimum=<rgba 0.478 0.478 0.478 1>, Maximum=<rgba 0.781 0.781 0.781 1>)
     ))
    ScaleRGBA=true
   ) {module=ScaleColor001}
   &ScaleColor(
    Scale Alpha=&Multiply_Float(
     A=&FloatFromCurve(
      FloatCurve=DataInterface() {propertyValues {Curve {keys [{interpMode=RCIM_Cubic, tangentMode=RCTM_User, time=0, value=0}, <at 0.15 1 RCIM_Cubic>, <at 0.9 0>]}}}
      Scale Curve=&FN_FrontEndFloatModifier(Gameplay=0.75, Front End=1)
     )
     B=&FloatFromCurve(
      FloatCurve=DataInterface() {propertyValues {Curve {keys [
          {interpMode=RCIM_Cubic, tangentMode=RCTM_User, time=0, value=0.6}
          {interpMode=RCIM_Cubic, tangentMode=RCTM_User, time=0.5, value=1}
          {tangentMode=RCTM_User, time=1, value=0.6}
         ]}}}
      CurveIndex=Linked() {linkedVariable {name=NPC.FortniteNPC.TimeOfDay_0-1, type=NiagaraFloat}}
     )
    )
    ScaleA=true
    ScaleRGB=false
   )
  ]
  rendererClasses [&NiagaraSpriteRendererProperties]
  renderers [
   {rendererClass=&NiagaraSpriteRendererProperties}
   {rendererClass=&NiagaraSpriteRendererProperties, rendererIndex=1}
  ]
 }
 FlashingLights {
  emitterUpdateScript [
   &EmitterState(
    Inactive Response=::NewEnumerator1
    Life Cycle Mode=::NewEnumerator1
    Loop Behavior=::NewEnumerator1
    Loop Duration Mode=::NewEnumerator1
    UseLoopDelay=false
   )
   &SpawnBurst_Instantaneous(Spawn Count=2, Use Spawn Probability=false, UseLoopCountLimit=false)
  ]
  particleSpawnScript [
   &PartitionParticles()
   &InitializeParticle(
    Color=&MaskLinearColor(
     Mask Parameter=$Particles.PartitionParticles.Partition
     Mask=0
     Masked Linear Color=<rgba 2 0.1666 0.1815 1>
     Passthrough Linear Color=<rgba 0.0434 0.1375 3 1>
    )
    Lifetime=0.8
    Position Offset=<0 -1 0>
    Sprite Size Mode=::NewEnumerator3
    Uniform Sprite Size=&FN_FrontEndFloatModifier(Gameplay=15, Front End=18)
    UsePositionOffset=true
    Write Material Random=true
   )
   &CameraOffset()
  ]
  particleUpdateScript [
   &ParticleState(
    Kill Particles When Lifetime Has Elapsed=false
    Let Infinitely Lived Particles Die When Emitter Deactivates=true
    Loop Particles Lifetime=true
   )
   &SkeletalMeshLocation(
    Bone / Socket Index=$Particles.PartitionParticles.Partition
    Bone / Socket Sampling Mode=::NewEnumerator5
    MaskSpawnGroup=false
    Mesh Sampling Type=::NewEnumerator6
    Sampled Position Offset=<0 0 -1>
    Skeletal Mesh=$System.LightLocations
   )
   &ScaleColor(
    Scale Alpha=&Multiply_Float(
     A=&Multiply_Float(
      A=&MaskFloat(
       Mask Parameter=$Particles.PartitionParticles.Partition
       Mask=0
       Masked Float=&FloatFromCurve(FloatCurve=DataInterface() {propertyValues {Curve {keys [<at 0 1>,<at 0.5 1>,<at 0.6 0>,<at 0.9 0>,<at 1 1>]}}})
       Passthrough Float=&FloatFromCurve(
        FloatCurve=DataInterface() {propertyValues {Curve {keys [<at 0 1>,<at 0.1 0>,<at 0.4 0>,<at 0.5 1>,<at 1 1>]}}}
        Scale Curve=&FN_FrontEndFloatModifier(Gameplay=1, Front End=2)
       )
      )
      B=&FloatFromCurve(
       FloatCurve=DataInterface() {propertyValues {Curve {keys [
           {arriveTangent=4.1785, interpMode=RCIM_Cubic, leaveTangent=4.1785, tangentMode=RCTM_User, time=0, value=0}
           {arriveTangent=1.1179, interpMode=RCIM_Cubic, leaveTangent=1.1179, tangentMode=RCTM_User, time=0.1822, value=0.6751}
           {arriveTangent=0.0224, interpMode=RCIM_Cubic, leaveTangent=0.0224, tangentMode=RCTM_User, time=0.5068, value=1}
          ]}}}
       CurveIndex=$Emitter.Age
      )
     )
     B=&Lerp_Float(
      A=0
      B=1
      Alpha=&Max_Float(
       A:NiagaraFloat=0
       B:NiagaraFloat=&Multiply_Float(
        A=&DotProduct(
         Vector A=&NormalizeVector(
          Vector To Normalize=&ReturnCameraVector(Camera Vectors=::NewEnumerator1, Coordinate Space=::World)
          Safe Normalize=true
          Safe Fallback Threshold=0
          Safe Fallback Vector=<0 0 0.001>
         )
         Vector B=&NormalizeVector(Vector To Normalize=$Engine.Owner.SystemYAxis, Safe Normalize=true, Safe Fallback Threshold=0, Safe Fallback Vector=<0 0 0.001>)
        )
        B=-1
       )
      )
     )
    )
    ScaleA=true
    ScaleRGB=true
   )
   &FN_Light_Attributes(
    Color=&MaskLinearColor(
     Mask Parameter=$Particles.PartitionParticles.Partition
     Mask=0
     Masked Linear Color=<rgba 4 0.0012 0 1>
     Passthrough Linear Color=<rgba 0.0465 0.4518 4 1>
    )
    Create Color Attribute=true
    Create Exponent Attribute=true
    Create Position Attribute=true
    Create Radius Attribute=true
    Enable Particle Light Enable=true
    Exponent=0.5
    Light Enabled=true
    Position=&AddVectorToPosition(
     Position=$Particles.SkeletalMeshLocation.SampledPosition
     Vector=&MaskVector(Mask Parameter=$Particles.PartitionParticles.Partition, Mask=0, Masked Vector=<-20 40 -60>, Passthrough Vector=<20 40 -60>)
    )
    Radius=80
    Scale Color=&Multiply_Float(
     A=&MaskFloat(
      Mask Parameter=$Particles.PartitionParticles.Partition
      Mask=0
      Masked Float=&FloatFromCurve(
       FloatCurve=DataInterface() {propertyValues {Curve {keys [<at 0 1>,<at 0.5 1>,<at 0.6 0>,<at 0.9 0>,<at 1 1>]}}}
       Scale Curve=&Multiply_Float(A=2, B=&FNTimeOfDayScalar(Night Scale=0.4, Morning Scale=0.6, Day Scale=0.8, Evening Scale=0.6))
      )
      Passthrough Float=&FloatFromCurve(
       FloatCurve=DataInterface() {propertyValues {Curve {keys [<at 0 1>,<at 0.1 0>,<at 0.4 0>,<at 0.5 1>,<at 1 1>]}}}
       Scale Curve=&Multiply_Float(A=3, B=&FNTimeOfDayScalar(Night Scale=0.7, Morning Scale=0.8, Day Scale=0.8, Evening Scale=0.8))
      )
     )
     B=&FloatFromCurve(
      FloatCurve=DataInterface() {propertyValues {Curve {keys [
          {arriveTangent=4.1785, interpMode=RCIM_Cubic, leaveTangent=4.1785, tangentMode=RCTM_User, time=0, value=0}
          {arriveTangent=1.1179, interpMode=RCIM_Cubic, leaveTangent=1.1179, tangentMode=RCTM_User, time=0.1822, value=0.6751}
          {arriveTangent=0.0224, interpMode=RCIM_Cubic, leaveTangent=0.0224, tangentMode=RCTM_User, time=0.5068, value=1}
         ]}}}
      CurveIndex=$Emitter.Age
     )
    )
   )
  ]
  rendererClasses [&NiagaraSpriteRendererProperties, &NiagaraLightRendererProperties]
  renderers [
   {rendererClass=&NiagaraSpriteRendererProperties}
   {rendererClass=&NiagaraLightRendererProperties, rendererIndex=1}
  ]
 }
 Headlights {
  emitterUpdateScript [
   &EmitterState(
    Inactive Response=::NewEnumerator1
    Life Cycle Mode=::NewEnumerator1
    Loop Behavior=::NewEnumerator1
    Loop Duration Mode=::NewEnumerator1
    UseLoopDelay=false
   )
   &SpawnBurst_Instantaneous(Spawn Count=2, Use Spawn Probability=false, UseLoopCountLimit=false)
  ]
  particleSpawnScript [
   &PartitionParticles()
   &InitializeParticle(
    Color=&MaskLinearColor(
     Mask Parameter=$Particles.PartitionParticles.Partition
     Mask=0
     Masked Linear Color=<rgba 1.5 1.1745 0.4197 0.171>
     Passthrough Linear Color=<rgba 1 0.0833 0.0907 0.542>
    )
    Lifetime=0
    Position Offset=<0 -1 0>
    Sprite Size Mode=::NewEnumerator3
    Uniform Sprite Size=30
    UsePositionOffset=true
    Write Material Random=true
   )
   &CameraOffset()
  ]
  particleUpdateScript [
   &ParticleState(Kill Particles When Lifetime Has Elapsed=false, Let Infinitely Lived Particles Die When Emitter Deactivates=true)
   &SkeletalMeshLocation(
    Bone / Socket Index=$Particles.PartitionParticles.Partition
    Bone / Socket Sampling Mode=::NewEnumerator5
    MaskSpawnGroup=false
    Mesh Sampling Type=::NewEnumerator6
    Sampled Position Offset=/Niagara/DynamicInputs/Helpers/MaskPosition(
     Mask Parameter:NiagaraInt32=$Particles.PartitionParticles.Partition
     Mask:NiagaraInt32=0
     Masked Position:NiagaraPosition=<0 10.366 -4.9145>
     Passthrough Position:NiagaraPosition=<0 -6.1111 0>
    )
    Skeletal Mesh=$System.HeadlightLocations
   )
   &ScaleColor(
    Scale Alpha=&MaskFloat(
     Mask Parameter=$Particles.PartitionParticles.Partition
     Mask=0
     Masked Float=&Multiply_Float(
      A=&Multiply_Float(
       A=&FN_FrontEndFloatModifier(
        Gameplay=&MaskFloat(Mask Parameter=$Particles.PartitionParticles.Partition, Mask=0, Masked Float=0.1, Passthrough Float=0.5)
        Front End=1
       )
       B=&FloatFromCurve(
        FloatCurve=DataInterface() {propertyValues {Curve {keys [
            {arriveTangent=4.1785, interpMode=RCIM_Cubic, leaveTangent=4.1785, tangentMode=RCTM_User, time=0, value=0}
            {arriveTangent=1.1179, interpMode=RCIM_Cubic, leaveTangent=1.1179, tangentMode=RCTM_User, time=0.1822, value=0.6751}
            {arriveTangent=0.0224, interpMode=RCIM_Cubic, leaveTangent=0.0224, tangentMode=RCTM_User, time=0.5068, value=1}
           ]}}}
        CurveIndex=$Emitter.Age
       )
      )
      B=&Lerp_Float(
       A=0
       B=1
       Alpha=&Max_Float(
        A:NiagaraFloat=0
        B:NiagaraFloat=&Multiply_Float(
         A=&DotProduct(
          Vector A=&NormalizeVector(
           Vector To Normalize=&ReturnCameraVector(Camera Vectors=::NewEnumerator1, Coordinate Space=::World)
           Safe Normalize=true
           Safe Fallback Threshold=0
           Safe Fallback Vector=<0 0 0.001>
          )
          Vector B=&NormalizeVector(Vector To Normalize=$Engine.Owner.SystemYAxis, Safe Normalize=true, Safe Fallback Threshold=0, Safe Fallback Vector=<0 0 0.001>)
         )
         B=-1
        )
       )
      )
     )
     Passthrough Float=1
    )
    ScaleA=true
    ScaleRGB=true
   )
  ]
  rendererClasses [&NiagaraSpriteRendererProperties]
  renderers [{rendererClass=&NiagaraSpriteRendererProperties}]
 }
}
summary {
 systemName=NS_Emote_GetawayCar_Police
 emitters [
  {emitterName=Default_Emote_TraversalSmoke_Loop, rendererClasses [&NiagaraSpriteRendererProperties]}
  {emitterName=Debris, rendererClasses [&NiagaraSpriteRendererProperties]}
  {
   emitterName=FlashingLights
   rendererClasses [&NiagaraSpriteRendererProperties, &NiagaraLightRendererProperties]
  }
  {emitterName=Headlights, rendererClasses [&NiagaraSpriteRendererProperties]}
 ]
 userVariables [
  User.bIsFrontEnd {type=NiagaraBool}
  User.bIsFrontEndPreview {type=NiagaraBool}
  User.bIsInWater {type=NiagaraBool}
  User.JetActiveCurve {type=NiagaraFloat, defaultValue=0, description=""}
 ]
}
