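# Niagara effect for the police car peel-out in the Getaway Car emote, built from two sprite emitters.
# Smoke spawns rotating puffs thrown along negative Y that drag, grow and partly follow source movement, with curve-driven dynamic material parameters.
# WindstreaksJet adds short-lived SubUV streaks from a burst plus steady rate, with emitter-age curves ramping size in and fading alpha out.
# Both emitters tint particles through randomized time-of-day colors.
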NiagaraSystem /BRCosmetics/Animation/Game/MainPlayer/Emotes/GetawayCar/FX/NS_Emote_GetawayCar_Police_PeelOut
dependencies {
 usedDataInterfaces [
  &NiagaraDataInterfaceCurve
  &NiagaraDataInterfaceSpriteRendererInfo
  &NiagaraDataInterfaceVector2DCurve
 ]
 usedDynamicInputs [&MakeFloatFromVector2d, &Multiply_Vector2]
 usedModules [&SystemState]
}
emitters {
 Smoke {
  emitterUpdateScript [
   &EmitterState(Life Cycle Mode=::NewEnumerator1, Loop Behavior=::NewEnumerator1, Loop Duration=1, UseLoopDelay=false)
   &SpawnRate(SpawnRate=8, Use Spawn Probability=false)
  ]
  particleSpawnScript [
   &InitializeParticle(
    Color=<rgba 0.5255 0.4735 0.4456 1>
    Lifetime=&UniformRangedFloat(Minimum=0.4, Maximum=0.8)
    Sprite Rotation Angle=&UniformRangedFloat(Minimum=0, Maximum=1)
    Sprite Rotation Mode=::NewEnumerator1
    Sprite Size Mode=::NewEnumerator3
    Uniform Sprite Size=&UniformRangedFloat(Minimum=25, Maximum=35)
    UsePositionOffset=false
    Write Material Random=true
   ) {module=InitializeParticle001}
   /SetVariables(Particles.InitSpriteRotRate=&UniformRangedFloat(Minimum=-1, Maximum=1)) {isSetParameters=true}
   &AccelerationForce(Acceleration=&UniformRangedVector(Minimum=<0 0 -25>, Maximum=<0 0 -75>))
   &SphereLocation(
    MaskSpawnGroup=false
    Override Seed=false
    Sphere Radius=10
    Surface Only=true
    UseNonUniformScale=false
    UseOffset=true
    UseSphereOrientation=false
   ) {enabled=false}
   &ApplyInitialForces()
   &CameraOffset(Camera Offset Amount=&RandomRangeFloat(Minimum=0, Maximum=15))
   &AddVelocity(Velocity=&RandomRangeVector(Minimum=<0 -400 100>, Maximum=<0 -200 50>))
  ]
  particleUpdateScript [
   &ParticleState()
   &SpriteRotationRate(Rotation Rate=&Multiply_Float(
     A=$Particles.InitSpriteRotRate
     B=&FloatFromCurve(
      FloatCurve=DataInterface() {propertyValues {Curve {keys [
          {arriveTangent=-1.9214, interpMode=RCIM_Cubic, leaveTangent=-1.9214, tangentMode=RCTM_User, time=-0.0046, value=1.0113}
          {arriveTangent=-0.2496, interpMode=RCIM_Cubic, leaveTangent=-0.2496, tangentMode=RCTM_User, time=0.9896, value=0.16}
         ]}}}
      CurveIndex=$Emitter.Age
     )
    ))
   &ScaleColor(
    Scale Mode=::NewEnumerator1
    Scale RGBA=&MakeVector4_FromLinearColor(LinearColor=&FNTimeOfDayColor(
      Night Color=&RandomRangeLinearColor(Minimum=<rgba 0.3893 0.4608 0.5318 0.9>, Maximum=<rgba 0.5702 0.6893 0.8605 1>)
      Morning Color=&RandomRangeLinearColor(Minimum=<rgba 0.2203 0.2814 0.6013 0.913>, Maximum=<rgba 0.3377 0.5872 0.9253 0.913>)
      Day Color=&RandomRangeLinearColor(Minimum=<rgba 0.478 0.4265 0.4121 0.7>, Maximum=<rgba 0.781 0.718 0.6943 1>)
      Evening Color=&RandomRangeLinearColor(Minimum=<rgba 0.478 0.478 0.478 1>, Maximum=<rgba 0.781 0.781 0.781 1>)
     ))
    ScaleRGBA=true
   )
   &ScaleColor(
    Scale Alpha=&FloatFromCurve(FloatCurve=DataInterface() {propertyValues {Curve {keys [<at 0 0>,<at 0.0483 0.9765>,<at 0.8887 0.9107>,<at 1 0>]}}})
    ScaleA=true
    ScaleRGB=false
   ) {module=ScaleColor001}
   &Drag(Drag=8, Use Linear Drag=true, Use Rotational Drag=false)
   &InheritSourceMovement(
    Applied Position Scale=&TransformVector(Vector=<0 0.7 0>, Source Space=::Simulation, Destination Space=::World)
    ApplyForce=false
    ApplyPosition=true
    ApplyPositionFalloff=false
   )
   &SolveForcesAndVelocity(Clamp Velocity=false, Limit Acceleration=false)
   &ScaleSpriteSize(Scale Factor=&Vector2DFromFloat(Value=&FloatFromCurve(FloatCurve=DataInterface() {propertyValues {Curve {keys [
          {arriveTangent=5.2354, interpMode=RCIM_Cubic, leaveTangent=5.2354, tangentMode=RCTM_User, time=-0.0046, value=0.25}
          <at 0.1439 1.4923>
          <at 1.0011 3.4809 RCIM_Cubic>
         ]}}})))
   &DynamicMaterialParameters(
    Index 0 Param 1=&FloatFromCurve(FloatCurve=DataInterface() {propertyValues {Curve {keys [<at 0 0 RCIM_Cubic>,<at 1.0011 0.2>]}}})
    Index 0 Param 2=&FloatFromCurve(FloatCurve=DataInterface() {propertyValues {Curve {keys [
         {arriveTangent=0.0453, interpMode=RCIM_Cubic, tangentMode=RCTM_Break, time=0, value=-1}
         {arriveTangent=0.0604, interpMode=RCIM_Cubic, leaveTangent=0.0604, tangentMode=RCTM_User, time=1, value=-1}
        ]}}})
    Index 0 Param 3=&FloatFromCurve(FloatCurve=DataInterface() {propertyValues {Curve {keys [{interpMode=RCIM_Cubic, leaveTangent=4.6779, time=0.2851, value=0}]}}})
    Index 1 Param 1=&FloatFromCurve(
     FloatCurve=DataInterface() {propertyValues {Curve {keys [<at 0 0>,<at 0.25 1>]}}}
     Scale Curve=&RandomRangeFloat(Minimum=2, Maximum=1)
    )
    Index 1 Param 2=&FloatFromCurve(
     FloatCurve=DataInterface() {propertyValues {Curve {keys [
         {interpMode=RCIM_Cubic, leaveTangent=3.7422, tangentMode=RCTM_Break, time=0, value=0.2}
         {interpMode=RCIM_Cubic, leaveTangent=0.9066, time=1, value=2}
        ]}}}
     Scale Curve=&RandomRangeFloat(Minimum=0.5, Maximum=1)
    )
    Index 1 Param 3=&FloatFromCurve(FloatCurve=DataInterface() {propertyValues {Curve {keys [
         {interpMode=RCIM_Cubic, leaveTangent=2.8445, tangentMode=RCTM_Break, time=0, value=0.2}
         {interpMode=RCIM_Cubic, leaveTangent=63.2978, time=0.25, value=1}
        ]}}})
    Index 1 Param 4=&FloatFromCurve(
     FloatCurve=DataInterface() {propertyValues {Curve {keys [<at 0 1>,<at 1 0.8>]}}}
     Scale Curve=&RandomRangeFloat(Minimum=0.5, Maximum=1)
    )
    Index 2 Param 1=&FloatFromCurve(FloatCurve=DataInterface() {propertyValues {Curve {keys [{interpMode=RCIM_Cubic, leaveTangent=3.2309, tangentMode=RCTM_Break, time=0.1, value=0}, <at 1 1>]}}})
    Index 2 Param 2=$Particles.MaterialRandom
    Index 2 Param 3=&FloatFromCurve(FloatCurve=DataInterface() {propertyValues {Curve {keys [
         {interpMode=RCIM_Cubic, tangentMode=RCTM_Break, time=0, value=1}
         {interpMode=RCIM_Cubic, leaveTangent=11.1233, tangentMode=RCTM_Break, time=1, value=2}
        ]}}})
    Index 2 Param 4=0.75
    Module.Param0WriteEnabled=true
    Module.Param0WriteEnabled1=true
    Module.Param0WriteEnabled2=true
    Module.Param1WriteEnabled=false
    Module.Param1WriteEnabled1=true
    Module.Param1WriteEnabled2=true
    Module.Param2WriteEnabled=true
    Module.Param2WriteEnabled1=true
    Module.Param2WriteEnabled2=true
    Module.Param3WriteEnabled=true
    Module.Param3WriteEnabled1=true
    Module.Param3WriteEnabled2=true
    Write Parameter Index 1=true
    Write Parameter Index 2=true
   )
  ]
  rendererClasses [&NiagaraSpriteRendererProperties]
  renderers [{rendererClass=&NiagaraSpriteRendererProperties}]
 }
 WindstreaksJet {
  emitterUpdateScript [
   &EmitterState(Life Cycle Mode=::NewEnumerator1, Loop Behavior=::NewEnumerator1, Loop Duration=1, UseLoopDelay=false)
   &SpawnRate(SpawnRate=20, Use Spawn Probability=false)
   &SpawnBurst_Instantaneous(Spawn Count=3, Use Spawn Probability=false, UseLoopCountLimit=false)
  ]
  particleSpawnScript [
   &InitializeParticle(
    Color=<rgba 0.5255 0.4735 0.4456 1>
    Lifetime=&UniformRangedFloat(Minimum=0.2, Maximum=0.4)
    Position Offset=<0 20 0>
    Sprite Size=&Multiply_Vector2D_ByFloat(Float=1, Vector2D=&UniformRangedVector2D(Minimum=<20 40>, Maximum=<30 50>))
    Sprite Size Mode=::NewEnumerator4
    UsePositionOffset=false
    Write Material Random=true
   )
   &AddVelocity(Velocity=&RandomRangeVector(Minimum=<50 -100 10>, Maximum=<-50 -200 25>), Velocity Speed Scale=0.1)
   &ApplyInitialForces()
   &SubUVAnimation(
    Evaluation Type=::NewEnumerator0
    "Sprite Renderer "=DataInterface() {propertyValues {SpriteRenderer=&self:WindstreaksJet_0.NiagaraSpriteRendererProperties_0}}
    SubUV Animation Mode:ENiagara_SubUVLookupModeV2=::NewEnumerator2
    UseEndFrame=false
    UseStartFrame=false
   )
  ]
  particleUpdateScript [
   &ParticleState()
   &ScaleSpriteSize(Scale Factor=&Multiply_Vector2D_ByFloat(
     Float=&FloatFromCurve(
      FloatCurve=DataInterface() {propertyValues {Curve {keys [
          {arriveTangent=10.1318, interpMode=RCIM_Cubic, leaveTangent=10.1318, tangentMode=RCTM_User, time=0, value=0}
          {interpMode=RCIM_Cubic, tangentMode=RCTM_User, time=0.2979, value=1}
         ]}}}
      CurveIndex=$Emitter.Age
     )
     Vector2D=&Vector2DFromCurve(
      Vector2Curve=DataInterface() {propertyValues {
        XCurve {keys [
          {arriveTangent=2.5428, interpMode=RCIM_Cubic, leaveTangent=2.5428, tangentMode=RCTM_User, time=0, value=1}
          {arriveTangent=0.1281, interpMode=RCIM_Cubic, leaveTangent=0.1281, tangentMode=RCTM_User, time=1, value=2}
         ]}
        YCurve {keys [
          {arriveTangent=2.8823, interpMode=RCIM_Cubic, leaveTangent=2.8823, tangentMode=RCTM_User, time=0, value=1}
          {arriveTangent=2.0297, interpMode=RCIM_Cubic, leaveTangent=2.0297, tangentMode=RCTM_User, time=1, value=2}
         ]}
       }}
      ScaleCurve=&Multiply_Vector2D_ByFloat(
       Float=&RandomRangeFloat(Minimum=1, Maximum=2)
       Vector2D=&RandomRangeVector2D(Minimum=<0.6 0.75>, Maximum=<0.8 1>)
      )
     )
    ))
   &DynamicMaterialParameters(
    Index 0 Param 1=5
    Index 0 Param 2=1
    Index 0 Param 3=1
    Index 0 Param 4=&FloatFromCurve(FloatCurve=DataInterface() {propertyValues {Curve {keys [{tangentMode=RCTM_User, time=0.1, value=0}, {tangentMode=RCTM_User, time=1, value=0}]}}})
    Module.Index0Float0Write=true
    Module.Index0Float1Write=true
    Module.Index0Float2Write=true
    Module.Index0Float3Write=true
   ) {module=DynamicMaterialParameters001}
   &Drag(Drag=5, Use Linear Drag=true, Use Rotational Drag=true)
   &SolveForcesAndVelocity(Clamp Velocity=false, Limit Acceleration=false)
   &ScaleColor(
    Scale Mode=::NewEnumerator1
    Scale RGBA=&MakeVector4_FromLinearColor(LinearColor=&FNTimeOfDayColor(
      Night Color=&RandomRangeLinearColor(Minimum=<rgba 0.0397 0.195 0.5318 0.4879>, Maximum=<rgba 0.112 0.3287 0.8605 0.3256>)
      Morning Color=&RandomRangeLinearColor(Minimum=<rgba 0.478 0.3753 0.3466 0.7>, Maximum=<rgba 0.781 0.6327 0.5772 0.4184>)
      Day Color=&RandomRangeLinearColor(Minimum=<rgba 0.478 0.3753 0.3466 0.7>, Maximum=<rgba 0.781 0.6327 0.5772 1>)
      Evening Color=&RandomRangeLinearColor(Minimum=<rgba 0.4041 0.478 0.4692 0.2947>, Maximum=<rgba 0.4268 0.4967 0.5249 0.1633>)
     ))
    ScaleRGBA=true
   )
   &ScaleColor(
    Scale Alpha=&FloatFromCurve(
     FloatCurve=DataInterface() {propertyValues {Curve {keys [<at 0 0>,<at 0.0483 0.9765>,<at 0.8887 0.9107>,<at 1 0>]}}}
     Scale Curve=&FloatFromCurve(
      FloatCurve=DataInterface() {propertyValues {Curve {keys [
          {tangentMode=RCTM_User, time=0, value=1}
          {arriveTangent=-0.0405, interpMode=RCIM_Cubic, leaveTangent=-0.0405, tangentMode=RCTM_User, time=0.7009, value=1}
          {arriveTangent=-8.0089, interpMode=RCIM_Cubic, leaveTangent=-8.0089, tangentMode=RCTM_User, time=1, value=0}
         ]}}}
      CurveIndex=$Emitter.Age
     )
    )
    ScaleA=true
    ScaleRGB=false
   ) {module=ScaleColor002}
  ]
  rendererClasses [&NiagaraSpriteRendererProperties]
  renderers [{rendererClass=&NiagaraSpriteRendererProperties}]
 }
}
summary {
 systemName=NS_Emote_GetawayCar_Police_PeelOut
 emitters [
  {emitterName=Smoke, rendererClasses [&NiagaraSpriteRendererProperties]}
  {emitterName=WindstreaksJet, rendererClasses [&NiagaraSpriteRendererProperties]}
 ]
 userVariables [
  User.JetActiveCurve {type=NiagaraFloat, defaultValue=0, description=""}
  User.NormalizedNotifyProgress {type=NiagaraFloat, defaultValue=0, description=""}
 ]
}
